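# Masked, two-sided, lit Niagara material for grassland particles on sprites, ribbons and meshes, with dandelion textures as defaults.
# A scalar splits the particle-tinted color texture between base color and emissive, with emissive passed through MF_FX_EmissiveAdjustments.
# Particle alpha, the opacity texture and a camera depth fade multiply into opacity, with an extra depth fade skipped on mobile.
# A normal map and metallic, roughness and specular parameters supply surface shading.
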Material /Game/Effects/Fort_Effects/Effects/Environment/Town/Materials/M_Grassland_Particles
graph {
 properties {
  bUsedWithNiagaraMeshParticles=true
  bUsedWithNiagaraRibbons=true
  bUsedWithNiagaraSprites=true
  blendMode=BLEND_Masked
  pixelDepthOffsetMode=PDOM_Legacy
  twoSided=true
  usedShadingModels=Default Lit
 }
 nodes [
  n1 = ParticleColor()
  n2 = TextureSampleParameter2D() {parameterName=BaseColor, texture=&/T_Dandelion_basecolor}
  n3 = ScalarParameter() {parameterName=Emissive Vs Base Color}
  MaterialFunctionCall(Emissive=Multiply(A=@n1.RGB, B=LinearInterpolate(B=@n2.RGB, Alpha=@n3))) {materialFunction=&MF_FX_EmissiveAdjustments}
  Multiply(A=@n1.RGB, B=LinearInterpolate(A=@n2.RGB, Alpha=@n3) {constB=0})
  ScalarParameter() {parameterName=Metallic}
  ScalarParameter() {defaultValue=0.65, parameterName=Roughness}
  ScalarParameter() {parameterName=Specular, sliderMin=0.5}
  ScalarParameter()
  n4 = TextureSampleParameter2D() {
   parameterName=Opacity
   samplerType=SAMPLERTYPE_LinearGrayscale
   texture=&/T_Dandelion_Opacity
  }
  n5 = Multiply(
   A=Multiply(A=@n1.A, B=@n4.R)
   B=MaterialFunctionCall(
    Fade Length=ScalarParameter() {defaultValue=100, parameterName=Fade Lenght}
    Fade Offset=ScalarParameter() {defaultValue=50, parameterName=Cmaera Fade Offset}
   ) {materialFunction=&CameraDepthFade}.Result
  )
  ShadingPathSwitch(Default=DepthFade(Opacity=@n5, FadeDistance=ScalarParameter() {parameterName=DepthFadeDistance}), Mobile=@n5)
  TextureSampleParameter2D() {parameterName=Normal, samplerType=SAMPLERTYPE_Normal, texture=&/T_Dandelion_N}
 ]
}
