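# Translucent, two-sided lit material for paper particles on Niagara sprites and meshes, using the Breakthrough emote papers texture.
# A SubUV sample tinted by particle color feeds overlay, emissive-adjustment and diffuse-conversion functions.
# A separate FlipBook picks a random cell from a 2x3 sheet, inverting the random value by face side so front and back can differ.
# Particle alpha is multiplied by a depth fade whose distance comes from a dynamic parameter.
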Material /Game/Effects/Fort_Effects/Effects/Environment/Town/Materials/M_Papers
graph {
 properties {
  bFullyRough=true
  bUsedWithNiagaraMeshParticles=true
  bUsedWithNiagaraSprites=true
  blendMode=BLEND_Translucent
  pixelDepthOffsetMode=PDOM_Legacy
  twoSided=true
  usedShadingModels=Default Lit
 }
 nodes [
  Constant()
  n1 = ParticleColor()
  n2 = Multiply(A=ParticleSubUV!(UVs, Tex, Apply View MipBias) {texture=/Game/Effects/Fort_Effects/Textures/T_Emote_Breakthrough_Papers}.RGB, B=@n1.RGB)
  MaterialFunctionCall(Base=@n2, Blend=DotProduct(A=TextureCoordinate(), B=Constant() {r=0.5})) {materialFunction=&Blend_Overlay}
  MaterialFunctionCall(Emissive=Multiply(A=@n2, B=@n1.A)) {materialFunction=&MF_FX_EmissiveAdjustments}
  MaterialFunctionCall(DiffuseColor=@n2) {materialFunction=&ConvertFromDiffSpec}
  Multiply(A=@n1.A, B=DepthFade(FadeDistance=DynamicParameter() {paramNames [DepthFadeDistance, Param2, Param3, Param4]}.DepthFadeDistance))
  n3 = ParticleRandom()
  n4 = OneMinus(None=@n3)
  n5 = Saturate(None=TwoSidedSign())
  n6 = Multiply(A=Subtract(A=TextureCoordinate()) {constB=0.5}, B=Constant2Vector() {g=0.99, r=0.95})
  SquareRoot(None=MaterialFunctionCall(
    "Animation  Phase (0-1)"=LinearInterpolate(A=@n3, B=@n4, Alpha=@n5)
    Number of Rows=Constant() {r=2}
    Number of Columns=Constant() {r=3}
    Texture=TextureObject() {texture=&T_Emote_Breakthrough_Papers}
    UVs=Add(A=@n6) {constB=0.5}
   ) {materialFunction=&FlipBook}.Result)
 ]
}
