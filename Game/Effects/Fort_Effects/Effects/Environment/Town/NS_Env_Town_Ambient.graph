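# Ambient Niagara system for town environments, drifting dandelion seeds and wind-blown paper through a user-scaled cylinder volume.
# Two active GPU emitters use user-set spawn probabilities and distance culling; Trash and DEBUG emitters are disabled.
# Seeds drift on wind and curl noise, scaling with camera distance.
# Paper spawns in looping gusts, snaps to terrain and depth buffer, culls bad placements, then tumbles with collision and camera repulsion.
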NiagaraSystem /Game/Effects/Fort_Effects/Effects/Environment/Town/NS_Env_Town_Ambient
dependencies {
 usedDataInterfaces [
  &NiagaraDataInterfaceCamera
  &NiagaraDataInterfaceArrayFloat
  &NiagaraDataInterfaceCurve
  &NiagaraDataInterfaceCollisionQuery
  &NiagaraDataInterfaceColorCurve
  &NiagaraDataInterfaceMeshRendererInfo
  &NiagaraDataInterfaceSpriteRendererInfo
 ]
 usedDynamicInputs []
 usedModules [
  &SystemState
  &self:SystemSpawnScript.SystemScriptSource.SystemScriptGraph.NiagaraNodeAssignment_0.SetVariables_#1
 ]
}
emitters {
 DEBUG {
  bEnabled=false
  simTarget=GPUComputeSim
  emitterUpdateScript [
   &EmitterState()
   &SpawnRate(Spawn Probability=$User.PaperSpawnProbability, SpawnRate=5, Use Spawn Probability=true)
  ]
  particleSpawnScript [
   &InitializeParticle(Color=<rgba 0.9188 0 1 1>, Lifetime=5, UsePositionOffset=false, Write Material Random=true)
   &ShapeLocation(
    Cylinder Height=200
    Cylinder Height Midpoint=0
    Cylinder Radius=&Lerp_Float(
     A=100
     B=500
     Alpha=&FloatFromCurve(
      FloatCurve=DataInterface() {propertyValues {Curve {keys [
          {interpMode=RCIM_Cubic, tangentMode=RCTM_User, time=0, value=0}
          {arriveTangentWeight=0.3508, interpMode=RCIM_Cubic, tangentMode=RCTM_User, time=1, value=1}
         ]}}}
      CurveIndex=&RandomRangeFloat(Minimum=0, Maximum=1)
     )
    )
    MaskSpawnGroup=false
    Non Uniform Scale=&Multiply_Vector(A=<1 1 1>, B=$User.VolumeScale)
    Override Seed=false
    Shape Primitive=::NewEnumerator1
    Surface Only=false
    UseOwnerScale=true
   )
  ]
  particleUpdateScript [&ParticleState()]
  rendererClasses [&NiagaraSpriteRendererProperties, &NiagaraMeshRendererProperties]
  renderers [
   {rendererClass=&NiagaraSpriteRendererProperties}
   {rendererClass=&NiagaraMeshRendererProperties, rendererIndex=1}
  ]
 }
 Dandilion_Seeds001 {
  simTarget=GPUComputeSim
  emitterSpawnScript [/SetVariables(Emitter.CurveDistance=1000) {isSetParameters=true}]
  emitterUpdateScript [
   &EmitterState(
    Enable Distance Culling=true
    Enable Visibility Culling=true
    MaxDistance:NiagaraFloat=&Multiply_Float(A=10000, B=$User.MaxDistanceScalar)
    Module.bUseMaxDistance=true
    Module.bUseMinDistance=true
    Scalability Mode=::NewEnumerator1
   )
   &SpawnRate(Spawn Probability=$User.DandelionProbability, SpawnRate=5, Use Spawn Probability=true)
   /SetVariables(Emitter.Cam Pos:NiagaraPosition=&ReturnCameraPosition(Coordinate Space:ENiagaraCoordinateSpace=::Simulation, Camera Query:NiagaraDataInterfaceCamera=DataInterface())) {isSetParameters=true}
  ]
  particleSpawnScript [
   &InitializeParticle(
    Lifetime Max=7.5
    Lifetime Min=3
    Lifetime Mode=::NewEnumerator1
    Mesh Scale=<0.75 0.75 0.5>
    Mesh Scale Mode=::NewEnumerator4
    Sprite Rotation Mode=::NewEnumerator2
    Sprite Size Mode=::NewEnumerator1
    Sprite UV Mode=::NewEnumerator3
    Uniform Sprite Size Max=16
    Uniform Sprite Size Min=13
    UsePositionOffset=false
    Write Material Random=true
   )
   &DynamicMaterialParameters(
    Index 0 Param 1=&SelectFloatFromArray(
     Array Sampling Mode=::NewEnumerator0
     Direct Array Index=?
     Float Selection Array=DataInterface() {propertyValues {FloatData [0,0.5,1]}}
     Random Seed=0
    )
    Module.Param0WriteEnabled=true
    Module.Param1WriteEnabled=true
    Module.Param2WriteEnabled=true
    Module.Param3WriteEnabled=true
   )
   &InitialMeshOrientation(Mesh Axis To Orient=<-1 0 0>, Mesh Orientation Mode=::NewEnumerator2, Orientation Vector=<0 1 0>, Use Rotation Vector=false)
   &SubUVAnimation(End Frame=3, SubUV Animation Mode:ENiagara_SubUVLookupMode=::NewEnumerator2) {module=SubUVAnimation001}
   /SetVariables(Particles.CameraOffset=6) {isSetParameters=true}
   &ShapeLocation(
    Cylinder Height=400
    Cylinder Height Midpoint=0
    Cylinder Radius=2500
    MaskSpawnGroup=false
    Override Seed=false
    Shape Primitive=::NewEnumerator1
    Surface Only=false
    UseOwnerScale=true
   ) {enabled=false}
   &ShapeLocation(
    Cylinder Height=400
    Cylinder Height Midpoint=0
    Cylinder Radius=&Lerp_Float(
     A=100
     B=500
     Alpha=&FloatFromCurve(
      FloatCurve=DataInterface() {propertyValues {Curve {keys [
          {interpMode=RCIM_Cubic, tangentMode=RCTM_User, time=0, value=0}
          {arriveTangent=0.934, interpMode=RCIM_Cubic, leaveTangent=0.934, tangentMode=RCTM_User, time=0.4, value=0.4}
          {arriveTangent=2.3845, interpMode=RCIM_Cubic, leaveTangent=2.3845, tangentMode=RCTM_User, time=1, value=1}
         ]}}}
      CurveIndex=&RandomRangeFloat(Minimum=0, Maximum=1)
     )
    )
    MaskSpawnGroup=false
    Non Uniform Scale=&Multiply_Vector(A=<1 1 1>, B=$User.VolumeScale)
    Override Seed=false
    Shape Primitive=::NewEnumerator1
    Surface Only=false
    UseOwnerScale=true
   ) {module=ShapeLocation001}
  ]
  particleUpdateScript [
   &ParticleState()
   &SpriteRotationRate(Rotation Rate=&Lerp_Float(
     A=0
     B=&RandomRangeFloat(Minimum=-60, Maximum=60)
     Alpha=&MakeFloatFromInt(Int=0)
    ))
   &UpdateMeshOrientation(
    "Foward Vector (Mesh Space)":Vector3f=<0 0 -1>
    Initial Forward Vector:Vector3f=<0 0 1>
    Orientation Method=::NewEnumerator3
    Particle Velocity:Vector3f=&Add_Vector(
     A=$Particles.Velocity
     B=&Lerp_Vector(
      A=<0 0 45>
      B=<0 0 75>
      Alpha=&Waveform(
       Global Input Position=&RandomRangeFloat(Minimum=0, Maximum=1)
       "Waveform [1]"=::NewEnumerator0
       "[1] Frequency"=1
       "[1] Amplitude Min/Max"=<0 1>
      )
     )
    )
    Up Vector:Vector3f=<1 0 0>
   )
   &AccelerationForce(Acceleration=&Lerp_Vector(
     A=<2 2 0>
     B=<50 50 2>
     Alpha=&FloatFromCurve(
      FloatCurve=DataInterface() {propertyValues {Curve {keys [
          {interpMode=RCIM_Cubic, tangentMode=RCTM_User, time=0, value=0}
          {arriveTangent=0.5713, interpMode=RCIM_Cubic, leaveTangent=0.5713, tangentMode=RCTM_User, time=0.8, value=0.2}
          {arriveTangent=6.6497, interpMode=RCIM_Cubic, leaveTangent=6.6497, tangentMode=RCTM_User, time=1, value=1}
         ]}}}
      CurveIndex=&RandomRangeFloat(Minimum=0, Maximum=1)
     )
    )) {enabled=false}
   &WindForce(Offset By Time:Vector3f=<0 0 1>, Scale=1, Wind Speed=<50 0 1>) {module=WindForce001}
   &Drag(Drag=&RandomRangeFloat(Minimum=0.8, Maximum=1.2), Use Linear Drag=true, Use Rotational Drag=false)
   &ScaleMeshSize(Scale Factor=&Multiply_VectorByFloat(
     Vector=&VectorFromFloat(Value=&FloatFromCurve(FloatCurve=DataInterface() {propertyValues {Curve {keys [
           {interpMode=RCIM_Cubic, tangentMode=RCTM_User, time=0, value=0}
           {interpMode=RCIM_Cubic, tangentMode=RCTM_User, time=0.1, value=1}
           {arriveTangentWeight=0.3986, interpMode=RCIM_Cubic, leaveTangentWeight=0.0498, tangentMode=RCTM_User, time=0.9, value=1}
           <at 1 0 RCIM_Cubic>
          ]}}}))
     Float=&FloatFromCurve(
      FloatCurve=DataInterface() {propertyValues {Curve {keys [<at 0.1 0>,<at 0.2 1>,<at 0.75 1>,<at 1 0>]}}}
      CurveIndex=&NormalizeDistanceRange(
       Start Position:NiagaraPosition=$Particles.Position
       End Position:NiagaraPosition=$Emitter.Cam Pos
       Distance:NiagaraFloat=$Emitter.CurveDistance
       Invert Normalized Range:NiagaraBool=false
      )
     )
    ))
   &ScaleSpriteSize(Scale Factor=&Multiply_Vector2D(
     A=&Vector2DFromFloat(Value=&FloatFromCurve(FloatCurve=DataInterface() {propertyValues {Curve {keys [
           <at 0 0.75 RCIM_Cubic>
           {arriveTangentWeight=0.0577, interpMode=RCIM_Cubic, leaveTangentWeight=0.3798, tangentMode=RCTM_User, time=0.112, value=1.0173}
           {arriveTangentWeight=0.3734, interpMode=RCIM_Cubic, leaveTangentWeight=0.0765, tangentMode=RCTM_User, time=0.85, value=1}
           {arriveTangent=-3.4635, interpMode=RCIM_Cubic, leaveTangent=-3.4635, tangentMode=RCTM_User, time=1.0011, value=0.75}
          ]}}}))
     B=&Vector2DFromFloat(Value=&FloatFromCurve(
       FloatCurve=DataInterface() {propertyValues {Curve {keys [{tangentMode=RCTM_User, time=0.75, value=1}, {tangentMode=RCTM_User, time=1, value=0}]}}}
       CurveIndex=&NormalizeDistanceRange(
        Start Position:NiagaraPosition=$Particles.Position
        End Position:NiagaraPosition=$Emitter.Cam Pos
        Distance:NiagaraFloat=$Emitter.CurveDistance
        Invert Normalized Range:NiagaraBool=false
       )
      ))
    ))
   &ScaleColor(
    Scale Alpha=&FloatFromCurve(FloatCurve=DataInterface() {propertyValues {Curve {keys [<at 0 0 RCIM_Cubic>,<at 0.5 1 RCIM_Cubic>,<at 1 0 RCIM_Cubic>]}}}, Scale Curve=0)
    ScaleA=true
    ScaleRGB=false
   )
   /SetVariables(Particles.Floof Color:LinearColor=&MultiplyLinearColor(
     Linear Color=&LerpLinearColors(
      StartColor=<rgba 0.4429 0.5 0.2446 0.1>
      EndColor=<rgba 0.9115 1 0.863 0.15>
      LerpFactor=&FloatFromCurve(
       FloatCurve=DataInterface() {propertyValues {Curve {keys [<at 0.5 1>,<at 0.85 0>]}}}
       CurveIndex=&NormalizeDistanceRange(
        Start Position:NiagaraPosition=$Particles.Position
        End Position:NiagaraPosition=$Emitter.Cam Pos
        Distance:NiagaraFloat=$Emitter.CurveDistance
        Invert Normalized Range:NiagaraBool=false
       )
      )
     )
     Scale Factor=&MakeFloat4FromVectorAndFloat(
      "Vector (RGB)"=<1 1 1>
      "Float (Alpha)"=&FloatFromCurve(FloatCurve=DataInterface() {propertyValues {Curve {keys [
           {interpMode=RCIM_Cubic, tangentMode=RCTM_User, time=0, value=0}
           {arriveTangentWeight=0.127, interpMode=RCIM_Cubic, leaveTangentWeight=0.3149, tangentMode=RCTM_User, time=0.25, value=1}
           {arriveTangentWeight=0.3444, interpMode=RCIM_Cubic, leaveTangentWeight=0.0722, tangentMode=RCTM_User, time=0.87, value=1}
           <at 1 0 RCIM_Cubic>
          ]}}})
     )
    )) {isSetParameters=true}
   &GravityForce(Gravity=&MakeCustomVectorFromBool(
     Bool=&SetBoolByIntComparison(A=0, B=1, Comparison Type=::NewEnumerator2)
     True Vector=<0 0 0>
     False Vector=&RandomRangeVector(Minimum=<0 0 1>, Maximum=<0 0 -5>)
    ))
   &CurlNoiseForce(
    Noise Frequency=25
    Noise Quality / Cost=::NewEnumerator0
    Noise Strength=&MakeCustomFloatFromBool(Bool=&SetBoolByIntComparison(A=0, B=1, Comparison Type=::NewEnumerator2), True Float=0, False Float=30)
    Pan Noise=true
    Pan Noise Field=<0 0 0.1>
   )
   &SolveForcesAndVelocity(Clamp Velocity=false, Limit Acceleration=false)
   &Color()
  ]
  rendererClasses [&NiagaraSpriteRendererProperties, &NiagaraMeshRendererProperties]
  renderers [
   {rendererClass=&NiagaraSpriteRendererProperties}
   {rendererClass=&NiagaraMeshRendererProperties, rendererIndex=1}
  ]
 }
 PaperInAir {
  simTarget=GPUComputeSim
  emitterUpdateScript [
   /SetVariables(
    Emitter.CameraPostion:Vector3f=&ReturnCameraVector(Camera Property=::NewEnumerator0, Coordinate Space=::Simulation)
    Emitter.Emitter_HeightOffset:NiagaraFloat=$System.System_HeightOffset
    Emitter.Overall_SpeedMult=0.2
    Emitter.SpawnRate_Curve:NiagaraFloat=&FloatFromCurve(
     FloatCurve=DataInterface() {propertyValues {Curve {keys [{arriveTangent=17.5837, time=0, value=0}, <at 0.1 1>, <at 0.2 0.25>, <at 0.6 0.6>, <at 0.9 0.2>, <at 1 0>]}}}
     CurveIndex=$Emitter.NormalizedLoopAge
    )
    Emitter.SpawnSpeed_Curve:NiagaraFloat=&FloatFromCurve(
     FloatCurve=DataInterface() {propertyValues {Curve {keys [
         {arriveTangent=30.2685, interpMode=RCIM_Cubic, leaveTangent=30.2685, tangentMode=RCTM_User, time=0.0009, value=0.0215}
         {arriveTangent=-0.0619, interpMode=RCIM_Cubic, leaveTangent=-0.0619, tangentMode=RCTM_User, time=0.0989, value=0.3515}
         <at 0.297 0.0032>
         <at 0.3192 0.3332>
         <at 0.3525 0.0398>
         <at 0.6171 0.3148>
         <at 0.9939 0.3698>
        ]}}}
     CurveIndex=&Multiply_Float(A=$Emitter.NormalizedLoopAge, B=&RandomRangeFloat(Minimum=1, Maximum=5))
     Scale Curve=&Multiply_Float(
      A=&Lerp_Float(A=5, B=10, Alpha=&UniformRangedFloat(Minimum=0.25, Maximum=1))
      B=1
     )
    )
    Emitter.SpriteRotRate:NiagaraFloat=&FloatFromCurve(
     FloatCurve=DataInterface() {propertyValues {Curve {keys [
         <at 0.0041 0.1315>
         <at 0.1088 1.9664>
         {arriveTangent=-3.3757, interpMode=RCIM_Cubic, leaveTangent=-3.3757, tangentMode=RCTM_User, time=0.3609, value=0.6391}
         {arriveTangent=-1, interpMode=RCIM_Cubic, leaveTangent=-1, tangentMode=RCTM_User, time=0.5614, value=0.4386}
         <at 0.7086 1.6472>
         <at 1 0>
        ]}}}
     CurveIndex=$Emitter.NormalizedLoopAge
     Scale Curve=100
    )
   ) {isSetParameters=true, module=SetVariables_#2}
   &EmitterState(
    Enable Distance Culling=true
    Life Cycle Mode=::NewEnumerator1
    Loop Duration=&RandomRangeFloat(Minimum=6, Maximum=10)
    MaxDistance:NiagaraFloat=&Multiply_Float(A=10000, B=$User.MaxDistanceScalar)
    Module.bApplySpawnCountScaleByDistance=true
    Module.bUseMaxDistance=true
    Module.bUseMinDistance=false
    Recalculate Duration Each Loop=true
    Scalability Mode=::NewEnumerator1
    Scale Spawn Count=true
    Spawn Count Scale By Distance=DataInterface() {propertyValues {Curve {keys [<at 0.75 0.9657>,<at 1 0.5>]}}}
    UseLoopDelay=false
   )
   &SpawnRate(
    Spawn Probability=$User.PaperSpawnProbability
    SpawnRate=&Multiply_Float(A=$Emitter.SpawnRate_Curve, B=5)
    Use Spawn Probability=true
   )
  ]
  particleSpawnScript [
   /SetVariables(
    Particles.HeightOffset=200
    Particles.UporDown:NiagaraFloat=&UniformAOrBFloat(
     A=&RandomRangeFloat(Minimum=12, Maximum=15)
     B=&RandomRangeFloat(Minimum=250, Maximum=1500)
     Distribution Weight=0.8
    )
   ) {isSetParameters=true, module=SetVariables_#3}
   /Niagara/Modules/Spawn/Initialization/InitializeParticle(
    Color=&RandomRangeLinearColor(Minimum=<rgba 0.2604 0.2604 0.2604 1>, Maximum=<rgba 0.8958 0.8958 0.8958 1>)
    Lifetime=&UniformRangedFloat(Minimum=3, Maximum=6)
    Mesh Scale=&Multiply_VectorByFloat(Vector=&RandomRangeVector(Minimum=<0.25 0.25 0.45>, Maximum=<0.4 0.4 0.6>), Float=8)
    Module.Write Color=true
    Module.Write Lifetime=true
    Module.Write Mass=true
    Module.Write Material Random=true
    Module.Write Position=false
    Module.Write Scale=true
    Module.Write SpriteRotation=true
    Module.Write SpriteSize=false
    Position=&SimulationPosition()
    Sprite Rotation=&UniformRangedFloat(Minimum=-180, Maximum=180)
    Sprite Size=&Multiply_Vector2D_ByFloat(
     Float=&Lerp_Float(
      A=0.75
      B=1.2
      Alpha=&FloatFromCurve(
       FloatCurve=DataInterface() {propertyValues {Curve {keys [
           {interpMode=RCIM_Cubic, tangentMode=RCTM_User, time=0, value=0}
           {arriveTangent=0.5713, interpMode=RCIM_Cubic, leaveTangent=0.5713, tangentMode=RCTM_User, time=0.8, value=0.2}
           {arriveTangent=6.6497, interpMode=RCIM_Cubic, leaveTangent=6.6497, tangentMode=RCTM_User, time=1, value=1}
          ]}}}
       CurveIndex=&RandomRangeFloat(Minimum=0, Maximum=1)
      )
     )
     Vector2D=&RandomRangeVector2D(Minimum=<10 10>, Maximum=<25 25>)
    )
   )
   &AddVelocity(Velocity=&Multiply_VectorByFloat(
     Vector=&Multiply_VectorByFloat(
      Vector=&Multiply_VectorByFloat(Vector=&UniformRangedVector(Minimum=<30 50 -250>, Maximum=<80 150 500>), Float=$Emitter.SpawnSpeed_Curve)
      Float=$Emitter.Overall_SpeedMult
     )
     Float=1
    ))
   &ShapeLocation(
    Cylinder Height=200
    Cylinder Height Midpoint=0
    Cylinder Radius=&Lerp_Float(
     A=100
     B=500
     Alpha=&FloatFromCurve(
      FloatCurve=DataInterface() {propertyValues {Curve {keys [
          {interpMode=RCIM_Cubic, tangentMode=RCTM_User, time=0, value=0}
          {arriveTangentWeight=1.9101, interpMode=RCIM_Cubic, tangentMode=RCTM_User, time=1, value=1}
         ]}}}
      CurveIndex=&RandomRangeFloat(Minimum=0, Maximum=1)
     )
    )
    MaskSpawnGroup=false
    Non Uniform Scale=&Multiply_Vector(A=<1 1 1>, B=$User.VolumeScale)
    Override Seed=false
    Shape Primitive=::NewEnumerator1
    Surface Only=false
    UseOwnerScale=true
   ) {module=ShapeLocation001}
   /NiagaraFN/Modules/Spawn/FN_SpawnParticlesOnTerrainNearPlayer(
    Height Texture -x -y:Vector3f=<-130048 -130048 0>
    HeightOffset=100
    HeightTexture +x +y:Vector3f=<130048 130048 0>
    Player Position:NiagaraPosition=$Particles.Position
   )
   /SetVariables(Particles.InitialPos:NiagaraPosition=$Particles.Position) {isSetParameters=true, module=SetVariables_#4}
   /Niagara/Modules/Math/PlaceParticlesOnDepthBuffer_GPU(Max Z Depth=5000, Module.Use Max Z Depth=true)
   &KillParticles(Kill Particles=&SetBoolByFloatComparison(
     A=&VectorLength(Vector=&Subtract_Vector(A=$Particles.InitialPos, B=Linked() {linkedVariable {name=Particles.Position, type=NiagaraPosition}}))
     B=500
    ))
   &KillParticles(Kill Particles=&Boolean_NotOperation(Bool=$Output.PlaceParticlesOnDepthBuffer_GPU.ValidSample)) {module=KillParticles001}
   &CurlNoiseForce(
    Noise Frequency=25
    Noise Strength=&Multiply_Float(
     A=250
     B=&Multiply_Float(A=&Multiply_Float(A=1, B=$Emitter.Overall_SpeedMult), B=1)
    )
    Pan Noise=false
   ) {module=CurlNoiseForce001}
   &ApplyInitialForces()
   /Niagara/Modules/Update/Camera/MaintainInCameraParticleScale()
   &SubUVAnimation(
    End Frame Range Override=0
    Evaluation Type=::NewEnumerator0
    "Sprite Renderer "=DataInterface() {propertyValues {SpriteRenderer=&self:PaperInAir_1.NiagaraSpriteRendererProperties_1}}
    SubUV Animation Mode:ENiagara_SubUVLookupModeV2=::NewEnumerator2
    UseEndFrame=false
    UseStartFrame=false
   ) {module=SubUVAnimation002}
   &SpriteFacingAndAlignment(Sprite Facing=<0 0 1>, WriteAlignment=false, WriteFacing=true)
   &DynamicMaterialParameters(
    Index 0 Param 1=&RandomRangeFloat(Minimum=0.6, Maximum=1.5)
    Index 0 Param 2=&RandomRangeFloat(Minimum=0.75, Maximum=1.5)
    Index 0 Param 3=&RandomRangeFloat(Minimum=1, Maximum=2.5)
    Index 0 Param 4=&RandomRangeFloat(Minimum=0.05, Maximum=0.2)
    Module.Param0WriteEnabled=true
    Module.Param1WriteEnabled=true
    Module.Param2WriteEnabled=true
    Module.Param3WriteEnabled=true
   )
   &InitialMeshOrientation(Use Rotation Vector=false)
  ]
  particleUpdateScript [
   &ParticleState()
   &ScaleMeshSize(Scale Factor=&VectorFromFloat(Value=&Multiply_Float(
      A=&FloatFromCurve(FloatCurve=DataInterface() {propertyValues {Curve {keys [
           {interpMode=RCIM_Cubic, tangentMode=RCTM_User, time=-0.0061, value=0.489}
           {arriveTangent=0.3677, interpMode=RCIM_Cubic, leaveTangent=0.3677, tangentMode=RCTM_User, time=0.2892, value=0.9107}
           {arriveTangent=-1.2812, interpMode=RCIM_Cubic, leaveTangent=-1.2812, tangentMode=RCTM_User, time=0.9809, value=0.9275}
           <at 1 0>
          ]}}})
      B=&FloatFromCurve(
       FloatCurve=DataInterface() {propertyValues {Curve {keys [
           {interpMode=RCIM_Cubic, tangentMode=RCTM_User, time=0, value=0.5}
           {arriveTangent=0.0074, interpMode=RCIM_Cubic, leaveTangent=0.0074, tangentMode=RCTM_User, time=0.15, value=1.0236}
           {arriveTangent=-0.0952, interpMode=RCIM_Cubic, leaveTangent=-0.0952, tangentMode=RCTM_User, time=0.85, value=1}
           <at 0.9 0>
          ]}}}
       CurveIndex=&NormalizeDistanceRange(
        Start Position:NiagaraPosition=Linked() {linkedVariable {name=Emitter.CameraPostion, type=Vector3f}}
        End Position:NiagaraPosition=$Particles.Position
        Distance:NiagaraFloat=10000
        Invert Normalized Range:NiagaraBool=false
       )
       Scale Curve=0.85
      )
     )))
   &Drag(
    Drag=&FloatFromCurve(FloatCurve=DataInterface() {propertyValues {Curve {keys [<at -0.0085 0.0857>,<at 0.0429 0.6046>,<at 0.3 0.125>,<at 1 0>]}}})
    Use Linear Drag=true
    Use Rotational Drag=false
   )
   &GravityForce(Gravity=&Multiply_VectorByFloat(
     Vector=<0 0 1>
     Float=/Game/Effects/Niagara/DynamicInputs/Math/FN_Fit(
      Clamp:NiagaraBool=false
      Value:NiagaraFloat=$Particles.UporDown
      SourceMin:NiagaraFloat=0
      SourceMax:NiagaraFloat=1
      DestMin:NiagaraFloat=0
      DestMax:NiagaraFloat=0.1
     )
    ))
   &CurlNoiseForce(
    Curl Noise Cone Mask Angle=50
    Mask Curl Noise=true
    Noise Frequency=&RandomRangeFloat(Minimum=500, Maximum=800)
    Noise Strength=&Multiply_Float(A=$Emitter.SpawnSpeed_Curve, B=1000)
    Pan Noise=true
    Pan Noise Field=<0 0 1>
    Random Seed=80
   )
   &PointAttractionForce(
    Attraction Radius=15000
    Attraction Strength=-150
    Attractor Position:Vector3f=$Emitter.CameraPostion
    Attractor Position Offset:Vector3f=<0 0 200>
    Falloff Exponent=0
    Kill Within Radius=false
    Module.Use Falloff=true
    UseAttractorOffset=true
   ) {module=PointAttractionForce001}
   &SpriteRotationRate(Rotation Rate=$Emitter.SpriteRotRate)
   &ScaleColor(
    Scale Alpha=&FloatFromCurve(FloatCurve=DataInterface() {propertyValues {Curve {keys [
         {interpMode=RCIM_Cubic, tangentMode=RCTM_User, time=0, value=0}
         {arriveTangent=-0.0948, interpMode=RCIM_Cubic, leaveTangent=-0.0948, tangentMode=RCTM_User, time=0.1, value=1}
         <at 0.942 0.9825>
         <at 1 0>
        ]}}})
    Scale RGB=&MakeVectorFromLinearColorRGB(Color=&HueShiftLinearColor(
      Hue=&RandomRangeFloat(Minimum=0, Maximum=0.1)
      Saturation=1
      Value=&RandomRangeFloat(Minimum=0.5, Maximum=2)
      Linear Color=$Particles.Initial.Color
     ))
    ScaleA=true
    ScaleRGB=true
   )
   &WindForce(Scale=1, Wind Speed=<200 0 -75>) {module=WindForce001}
   &Collision(
    Age Colliding Particles=false
    Enable Rest State=false
    Kill Offscreen Particles=true
    Module.EnableMaxCollisionCount=false
    Module.Randomize Collision Normal Vector=false
    Particle Radius Scale=2
    Restitution=0.7
   )
   &AerodynamicDrag(Aerodynamic Drag=0.5, Use Linear Drag=true, Use Rotational Drag=true)
   &SolveForcesAndVelocity(Clamp Velocity=false, Limit Acceleration=false)
   &AlignSpriteToMeshOrientation(SetAlignment=true, SetFacing=true)
   &UpdateMeshOrientation(
    Rotation Rate=&FloatFromCurve(FloatCurve=DataInterface() {propertyValues {Curve {keys [
         {interpMode=RCIM_Cubic, tangentMode=RCTM_User, time=0, value=0}
         {interpMode=RCIM_Cubic, tangentMode=RCTM_User, time=0.5, value=1}
         {interpMode=RCIM_Cubic, tangentMode=RCTM_User, time=1, value=0}
        ]}}})
    Rotation Vector=&RandomRangeVector(Minimum=<0.1 0.1 0.1>, Maximum=<1 1 1>)
   )
  ]
  rendererClasses [&NiagaraMeshRendererProperties]
  renderers [{rendererClass=&NiagaraMeshRendererProperties, rendererIndex=0}]
 }
 Trash {
  bEnabled=false
  simTarget=GPUComputeSim
  emitterUpdateScript [
   /SetVariables(
    Emitter.CameraPostion:Vector3f=&ReturnCameraVector(Camera Property=::NewEnumerator0, Coordinate Space=::Simulation)
    Emitter.Emitter_HeightOffset:NiagaraFloat=$System.System_HeightOffset
    Emitter.Overall_SpeedMult=0.2
    Emitter.SpawnRate_Curve:NiagaraFloat=&FloatFromCurve(
     FloatCurve=DataInterface() {propertyValues {Curve {keys [{arriveTangent=17.5837, time=0, value=0}, <at 0.1 1>, <at 0.2 0.25>, <at 0.6 0.6>, <at 0.9 0.2>, <at 1 0>]}}}
     CurveIndex=$Emitter.NormalizedLoopAge
    )
    Emitter.SpawnSpeed_Curve:NiagaraFloat=&FloatFromCurve(
     FloatCurve=DataInterface() {propertyValues {Curve {keys [
         {arriveTangent=30.2685, interpMode=RCIM_Cubic, leaveTangent=30.2685, tangentMode=RCTM_User, time=0.0009, value=0.0215}
         {arriveTangent=-0.0619, interpMode=RCIM_Cubic, leaveTangent=-0.0619, tangentMode=RCTM_User, time=0.0989, value=0.3515}
         <at 0.297 0.0032>
         <at 0.3192 0.3332>
         <at 0.3525 0.0398>
         <at 0.6171 0.3148>
         <at 0.9939 0.3698>
        ]}}}
     CurveIndex=&Multiply_Float(A=$Emitter.NormalizedLoopAge, B=&RandomRangeFloat(Minimum=1, Maximum=5))
     Scale Curve=&Multiply_Float(
      A=&Lerp_Float(A=5, B=10, Alpha=&UniformRangedFloat(Minimum=0.25, Maximum=1))
      B=1
     )
    )
    Emitter.SpriteRotRate:NiagaraFloat=&FloatFromCurve(
     FloatCurve=DataInterface() {propertyValues {Curve {keys [
         <at 0.0041 0.1315>
         <at 0.1088 1.9664>
         {arriveTangent=-3.3757, interpMode=RCIM_Cubic, leaveTangent=-3.3757, tangentMode=RCTM_User, time=0.3609, value=0.6391}
         {arriveTangent=-1, interpMode=RCIM_Cubic, leaveTangent=-1, tangentMode=RCTM_User, time=0.5614, value=0.4386}
         <at 0.7086 1.6472>
         <at 1 0>
        ]}}}
     CurveIndex=$Emitter.NormalizedLoopAge
     Scale Curve=100
    )
   ) {isSetParameters=true, module=SetVariables_#2}
   &EmitterState(
    Enable Distance Culling=true
    Life Cycle Mode=::NewEnumerator1
    Loop Duration=&RandomRangeFloat(Minimum=6, Maximum=10)
    MaxDistance=10000
    Module.bApplySpawnCountScaleByDistance=true
    Module.bUseMaxDistance=true
    Module.bUseMinDistance=false
    Recalculate Duration Each Loop=true
    Scalability Mode=::NewEnumerator1
    Scale Spawn Count=true
    Spawn Count Scale By Distance=DataInterface() {propertyValues {Curve {keys [<at 0.75 0.9657>,<at 1 0.5>]}}}
    UseLoopDelay=false
   )
   &SpawnRate(
    Spawn Probability=$User.CupSpawnProbability
    SpawnRate=&Multiply_Float(A=$Emitter.SpawnRate_Curve, B=1)
    Use Spawn Probability=true
   )
  ]
  particleSpawnScript [
   /SetVariables(
    Particles.HeightOffset=200
    Particles.UporDown:NiagaraFloat=&UniformAOrBFloat(
     A=&RandomRangeFloat(Minimum=12, Maximum=15)
     B=&RandomRangeFloat(Minimum=50, Maximum=75)
     Distribution Weight=0.8
    )
   ) {isSetParameters=true, module=SetVariables_#3}
   &InitializeParticle(
    Color=&MultiplyLinearColor(
     Linear Color=&ColorFromCurve(
      ColorCurve=DataInterface() {propertyValues {
        AlphaCurve {keys [<at 0 1 RCIM_Cubic>,<at 1 1 RCIM_Cubic>]}
        BlueCurve {keys [<at 0 0.0172 RCIM_Constant>,<at 0.33 0.0109 RCIM_Constant>,<at 0.66 0.4583 RCIM_Constant>]}
        GreenCurve {keys [<at 0 0.0172 RCIM_Constant>,<at 0.33 0.2188 RCIM_Constant>,<at 0.66 0.0613 RCIM_Constant>]}
        RedCurve {keys [<at 0 0.3438 RCIM_Constant>,<at 0.33 0.0602 RCIM_Constant>,<at 0.66 0.0229 RCIM_Constant>]}
       }}
      CurveIndex=&RandomRangeFloat(Minimum=0, Maximum=1)
     )
     Scale Factor=<1 1 1 1>
    )
    Lifetime=&UniformRangedFloat(Minimum=3, Maximum=4)
    Mesh Scale=&Lerp_Vector(
     A=<0.5 0.5 1.1>
     B=<0.5 0.5 1.5>
     Alpha=&MakeCustomFloatFromBool(Bool=&RandomBool(), True Float=1, False Float=0)
    )
    Module.Write Color=true
    Module.Write Lifetime=true
    Module.Write Mass=true
    Module.Write Material Random=true
    Module.Write Position=false
    Module.Write Scale=true
    Module.Write SpriteRotation=true
    Module.Write SpriteSize=false
    Position=&SimulationPosition()
    Sprite Rotation=&UniformRangedFloat(Minimum=-180, Maximum=180)
    Sprite Size=&Multiply_Vector2D_ByFloat(
     Float=&Lerp_Float(
      A=0.75
      B=1.2
      Alpha=&FloatFromCurve(
       FloatCurve=DataInterface() {propertyValues {Curve {keys [
           {interpMode=RCIM_Cubic, tangentMode=RCTM_User, time=0, value=0}
           {arriveTangent=0.5713, interpMode=RCIM_Cubic, leaveTangent=0.5713, tangentMode=RCTM_User, time=0.8, value=0.2}
           {arriveTangent=6.6497, interpMode=RCIM_Cubic, leaveTangent=6.6497, tangentMode=RCTM_User, time=1, value=1}
          ]}}}
       CurveIndex=&RandomRangeFloat(Minimum=0, Maximum=1)
      )
     )
     Vector2D=&RandomRangeVector2D(Minimum=<10 10>, Maximum=<25 25>)
    )
   )
   &AddVelocity(
    Scale Added Velocity=<1 1 0.25>
    Velocity=&Multiply_VectorByFloat(
     Vector=&Multiply_VectorByFloat(
      Vector=&Multiply_VectorByFloat(Vector=&UniformRangedVector(Minimum=<75 50 0>, Maximum=<200 150 0>), Float=$Emitter.SpawnSpeed_Curve)
      Float=$Emitter.Overall_SpeedMult
     )
     Float=1
    )
   )
   &ShapeLocation(
    Cylinder Height=2
    Cylinder Height Midpoint=0
    Cylinder Radius=&Lerp_Float(
     A=100
     B=500
     Alpha=&FloatFromCurve(
      FloatCurve=DataInterface() {propertyValues {Curve {keys [
          {interpMode=RCIM_Cubic, tangentMode=RCTM_User, time=0, value=0}
          {arriveTangent=0.934, interpMode=RCIM_Cubic, leaveTangent=0.934, tangentMode=RCTM_User, time=0.8, value=0.25}
          {arriveTangent=5.6424, interpMode=RCIM_Cubic, leaveTangent=5.6424, tangentMode=RCTM_User, time=1, value=1}
         ]}}}
      CurveIndex=&RandomRangeFloat(Minimum=0, Maximum=1)
     )
    )
    MaskSpawnGroup=false
    Non Uniform Scale=&Multiply_Vector(A=<1 1 1>, B=$User.VolumeScale)
    Override Seed=false
    Shape Primitive=::NewEnumerator1
    Surface Only=false
    UseOwnerScale=true
   ) {module=ShapeLocation001}
   &FN_SpawnParticlesOnTerrainNearPlayer(
    Height Texture -x -y:Vector3f=<-130048 -130048 0>
    HeightOffset=10
    HeightTexture +x +y:Vector3f=<130048 130048 0>
    Player Position:NiagaraPosition=$Particles.Position
   )
   /SetVariables(Particles.InitialPos:NiagaraPosition=$Particles.Position) {isSetParameters=true, module=SetVariables_#4}
   &PlaceParticlesOnDepthBuffer_GPU(Max Z Depth=5000, Module.Use Max Z Depth=true)
   &KillParticles(Kill Particles=&SetBoolByFloatComparison(
     A=&VectorLength(Vector=&Subtract_Vector(A=$Particles.InitialPos, B=Linked() {linkedVariable {name=Particles.Position, type=NiagaraPosition}}))
     B=500
    ))
   &KillParticles(Kill Particles=&Boolean_NotOperation(Bool=$Output.PlaceParticlesOnDepthBuffer_GPU.ValidSample)) {module=KillParticles001}
   &ApplyInitialForces()
   &MaintainInCameraParticleScale()
   &SpriteFacingAndAlignment(Sprite Facing=<0 0 1>, WriteAlignment=false, WriteFacing=true)
   &DynamicMaterialParameters(
    Index 0 Param 1=20
    Module.Param0WriteEnabled=true
    Module.Param1WriteEnabled=true
    Module.Param2WriteEnabled=true
    Module.Param3WriteEnabled=true
   )
   &InitialMeshOrientation(Use Rotation Vector=false) {enabled=false}
  ]
  particleUpdateScript [
   &ParticleState()
   &ScaleMeshSize(Scale Factor=&VectorFromFloat(Value=&Multiply_Float(
      A=&FloatFromCurve(FloatCurve=DataInterface() {propertyValues {Curve {keys [
           {interpMode=RCIM_Cubic, tangentMode=RCTM_User, time=-0.0061, value=0.489}
           {arriveTangent=0.3677, interpMode=RCIM_Cubic, leaveTangent=0.3677, tangentMode=RCTM_User, time=0.2892, value=0.9107}
           {arriveTangent=-1.2812, interpMode=RCIM_Cubic, leaveTangent=-1.2812, tangentMode=RCTM_User, time=0.9809, value=0.9275}
           <at 1 0>
          ]}}})
      B=&FloatFromCurve(
       FloatCurve=DataInterface() {propertyValues {Curve {keys [
           {interpMode=RCIM_Cubic, tangentMode=RCTM_User, time=0, value=0}
           {arriveTangent=0.0074, interpMode=RCIM_Cubic, leaveTangent=0.0074, tangentMode=RCTM_User, time=0.15, value=1.0236}
           {arriveTangent=-0.0952, interpMode=RCIM_Cubic, leaveTangent=-0.0952, tangentMode=RCTM_User, time=0.85, value=1}
           <at 0.9 0>
          ]}}}
       CurveIndex=&NormalizeDistanceRange(
        Start Position:NiagaraPosition=Linked() {linkedVariable {name=Emitter.CameraPostion, type=Vector3f}}
        End Position:NiagaraPosition=$Particles.Position
        Distance:NiagaraFloat=10000
        Invert Normalized Range:NiagaraBool=false
       )
       Scale Curve=0.85
      )
     )))
   &Drag(
    Drag=&FloatFromCurve(FloatCurve=DataInterface() {propertyValues {Curve {keys [<at 0 0.3>,<at 0.1 0.8>,<at 0.7 1>,<at 1 0.2>]}}}, Scale Curve=0.1)
    Use Linear Drag=true
    Use Rotational Drag=false
   )
   &GravityForce(Gravity=<0 0 -400>)
   &SpriteRotationRate(Rotation Rate=$Emitter.SpriteRotRate)
   &ScaleColor(
    Scale Alpha=&FloatFromCurve(FloatCurve=DataInterface() {propertyValues {Curve {keys [
         {interpMode=RCIM_Cubic, tangentMode=RCTM_User, time=0, value=0}
         {arriveTangent=-0.0948, interpMode=RCIM_Cubic, leaveTangent=-0.0948, tangentMode=RCTM_User, time=0.1, value=1}
         <at 0.942 0.9825>
         <at 1 0>
        ]}}})
    Scale RGB=&MakeVectorFromLinearColorRGB(Color=&HueShiftLinearColor(
      Hue=&RandomRangeFloat(Minimum=0, Maximum=0.1)
      Saturation=1
      Value=&RandomRangeFloat(Minimum=0.5, Maximum=2)
      Linear Color=$Particles.Initial.Color
     ))
    ScaleA=true
    ScaleRGB=true
   )
   &WindForce(Offset By Time:Vector3f=<100 0 -50>, Scale=1, Wind Speed=<100 20 -100>) {module=WindForce001}
   &Collision(
    Age Colliding Particles=false
    Enable Rest State=false
    Friction=0.1
    Kill Offscreen Particles=true
    Module.EnableMaxCollisionCount=false
    Module.Randomize Collision Normal Vector=false
    Particle Radius Scale=2
    Restitution=0.4
   )
   &AerodynamicDrag(
    Lift Contribution=0.1
    Pivot Offset:Vector3f=<0.2 0.2 -1>
    Rotation Strength=0.1
    Rotational Pivot Mode:ENiagaraAerodynamicDragPivotMode=::NewEnumerator1
    Use Linear Drag=true
    Use Rotational Drag=true
   )
   &SolveForcesAndVelocity(Clamp Velocity=false, Limit Acceleration=false)
   &AlignSpriteToMeshOrientation(SetAlignment=true, SetFacing=true)
   &DynamicMaterialParameters(
    Index 0 Param 3=&FloatFromCurve(FloatCurve=DataInterface() {propertyValues {Curve {keys [
         {interpMode=RCIM_Cubic, tangentMode=RCTM_User, time=0, value=0}
         {arriveTangent=-0.0948, interpMode=RCIM_Cubic, leaveTangent=-0.0948, tangentMode=RCTM_User, time=0.1, value=1}
         <at 0.942 0.9825>
         <at 1 0>
        ]}}})
    Module.Param0WriteEnabled=false
    Module.Param1WriteEnabled=false
    Module.Param2WriteEnabled=true
    Module.Param3WriteEnabled=true
   ) {module=DynamicMaterialParameters001}
   &UpdateMeshOrientation(Rotation Rate=&FloatFromCurve(FloatCurve=DataInterface() {propertyValues {Curve {keys [
         {interpMode=RCIM_Cubic, tangentMode=RCTM_User, time=0, value=0}
         {interpMode=RCIM_Cubic, tangentMode=RCTM_User, time=0.5, value=1}
         {interpMode=RCIM_Cubic, tangentMode=RCTM_User, time=1, value=0}
        ]}}}))
  ]
  rendererClasses [&NiagaraMeshRendererProperties]
  renderers [
   {rendererClass=&NiagaraMeshRendererProperties, rendererIndex=0}
   {rendererClass=&NiagaraMeshRendererProperties, rendererIndex=1}
  ]
 }
}
summary {
 systemName=NS_Env_Town_Ambient
 emitters [
  {
   emitterName=Dandilion_Seeds001
   simTarget=GPUComputeSim
   rendererClasses [&NiagaraSpriteRendererProperties, &NiagaraMeshRendererProperties]
  }
  {
   bEnabled=false
   emitterName=DEBUG
   simTarget=GPUComputeSim
   rendererClasses [&NiagaraSpriteRendererProperties, &NiagaraMeshRendererProperties]
  }
  {bEnabled=false, emitterName=Trash, simTarget=GPUComputeSim, rendererClasses [&NiagaraMeshRendererProperties]}
  {emitterName=PaperInAir, simTarget=GPUComputeSim, rendererClasses [&NiagaraMeshRendererProperties]}
 ]
 userVariables [
  User.CupSpawnProbability {type=NiagaraFloat, defaultValue=0, description=""}
  User.DandelionProbability {type=NiagaraFloat, defaultValue=0.1, description=""}
  User.DebugOn {type=NiagaraBool}
  User.Landscape {
   type=NiagaraDataInterfaceLandscape
   description=""
   defaultValue {
    struct=NiagaraExt_VariableValue_DataInterface
    value {
     dataInterface=&self:NiagaraDataInterfaceLandscape_0
     dataInterfaceClass=&NiagaraDataInterfaceLandscape
    }
   }
  }
  User.MaxDistanceScalar {type=NiagaraFloat, defaultValue=1, description=""}
  User.PaperSpawnProbability {type=NiagaraFloat, defaultValue=1, description=""}
  User.TextureCup {
   type=Texture
   description=""
   defaultValue {
    struct=NiagaraExt_VariableValue_Object
    value {object=/Game/Global/Textures/Default/Blanks/T_WhiteUserInterface, objectClass=/Script/Engine.Texture}
   }
  }
  User.VolumeScale {type=Vector3f, description="", defaultValue {struct=Vector3f, value=<1 1 1>}}
 ]
}
